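; BP_MovingCube is an AimLab test-map actor whose static mesh slides back and forth along X around its starting location.
; BeginPlay stores the mesh's start location, then UpdatePos re-arms itself with a next-tick timer instead of using Event Tick.
; Each update derives a ping-pong alpha from server world time divided by Speed, lerps between plus and minus Dist, and teleports the mesh, so position depends on shared time rather than accumulated movement.
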(event EventBeginPlay
  (Variables|Default|SetStartLocation (Transformation|GetWorldLocation (Variables|Debug|GetStaticMeshComponent)))
  (Utilities|Time|SetTimerforNextTickbyEvent (EventDispatchers|CreateEvent :!event Custom|UpdatePos)))

(event Collision|EventActorBeginOverlap
  (params (OtherActor :type Actor%20Object%20Reference)))

(event EventTick
  (params (DeltaSeconds :type Float%20%28single-precision%29)))

(event Custom|UpdatePos
  (bind (_startlocation_x _startlocation_y _startlocation_z) (Math|Vector|BreakVector (Variables|GetStartLocation_X)))
  (Transformation|SetWorldLocation (Variables|Debug|GetStaticMeshComponent) (Math|Vector|MakeVector (+ (Math|Float|Lerp (Variables|Default|GetDist) (* (Variables|Default|GetDist) -1.0) (Math|Float|SelectFloat (- 1.0 (Math|Float|%25%28Float%29 (/ (Fortnite|Sync|Time|GetServerWorldTimeSeconds) (Variables|Default|GetSpeed)))) (Math|Float|%25%28Float%29 (/ (Fortnite|Sync|Time|GetServerWorldTimeSeconds) (Variables|Default|GetSpeed))) (> (Math|Float|%25%28Float%29 (/ (Fortnite|Sync|Time|GetServerWorldTimeSeconds) (Variables|Default|GetSpeed)) 2.0) 1.0))) _startlocation_x) _startlocation_y _startlocation_z) false true)
  (Utilities|Time|SetTimerforNextTickbyEvent (EventDispatchers|CreateEvent :!event Custom|UpdatePos)))

(fn ConstructionScript)
